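# Niagara system for the Solstice backpack cosmetic that adds a persistent glowing radial flare with a light.
# Its single RadialFlare emitter bursts one offset particle, drawn by mesh, sprite and light renderers, that lives until the emitter deactivates.
# A curve indexed by the Fortnite time-of-day parameter varies the alpha of a DarkenColor attribute.
# A sine of emitter age pulses the reddish light's color scale.
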NiagaraSystem /BRCosmetics/Accessories/FORT_Backpacks/Backpack_Solstice/FX/NS_Solstice_Backpack
dependencies {usedDataInterfaces [&NiagaraDataInterfaceCurve], usedDynamicInputs [], usedModules [&SystemState]}
emitters {RadialFlare {
  emitterUpdateScript [
   &EmitterState(
    Life Cycle Mode=::NewEnumerator1
    Loop Behavior=::NewEnumerator1
    Loop Duration Mode=::NewEnumerator1
    Reset Age On Awaken=true
    Scalability Mode=::NewEnumerator1
    UseLoopDelay=false
   )
   &SpawnBurst_Instantaneous(Spawn Count=1, Use Spawn Probability=false)
  ]
  particleSpawnScript [
   /SetVariables(Particles.New Position:NiagaraFloat=&VectorLength(Vector=<0 0 0>)) {isSetParameters=true}
   /Niagara/Modules/Spawn/Initialization/InitializeParticle(
    Color=<rgba 1.4621 3.0902 4 0.1>
    Mesh Scale=&VectorFromFloat(Value=1.08)
    Module.Write Color=true
    Module.Write Lifetime=true
    Module.Write Mass=true
    Module.Write Material Random=true
    Module.Write Position=true
    Module.Write Scale=true
    Module.Write SpriteRotation=false
    Module.Write SpriteSize=true
    Position=&SimulationPosition()
    Sprite Size=&Vector2DFromFloat(Value=30)
   ) {enabled=false}
   /Niagara/Modules/Spawn/Initialization/V2/InitializeParticle(
    Color=<rgba 1.4621 3.0902 4 0.1>
    Mesh Scale Mode=::NewEnumerator3
    Mesh Uniform Scale=1.08
    Position Offset=<0 8 0>
    Sprite Size Mode=::NewEnumerator3
    Uniform Sprite Size=30
    UsePositionOffset=true
    Write Material Random=true
   ) {module=InitializeParticle001}
   &DynamicMaterialParameters(
    Index 0 Param 1=$Particles.MaterialRandom
    Index 0 Param 3=1
    Module.Param0WriteEnabled=true
    Module.Param1WriteEnabled=true
    Module.Param2WriteEnabled=true
    Module.Param3WriteEnabled=false
   )
   /SetVariables(Particles.DarkenDynamicParams:Vector4f=<1 1 1 0>) {isSetParameters=true}
   &InitialMeshOrientation(Orientation Vector=<0 -1 0>, Use Orientation Vector=true, Use Rotation Vector=false) {module=InitialMeshOrientation001}
  ]
  particleUpdateScript [
   /SetVariables(Particles.DarkenColor:LinearColor=&MakeLinearColorFromVectorAndFloat(
     "Vector (RGB)"=&MakeVectorFromLinearColorRGB(Color=<rgba 1 0.2488 0.2289 1>)
     "Float (Alpha)"=&Lerp_Float(
      A=0.25
      B=0.15
      Alpha=&FloatFromCurve(
       FloatCurve=DataInterface() {propertyValues {Curve {keys [
           <at 0 0>
           <at 0.1667 1>
           {arriveTangent=-1.7143, leaveTangent=-1.7143, time=0.25, value=0}
           {arriveTangent=1.6, leaveTangent=1.6, time=0.75, value=0}
           {arriveTangent=-17.5595, interpMode=RCIM_Cubic, leaveTangent=-17.5593, tangentMode=RCTM_User, time=0.875, value=1.1}
           {arriveTangent=-1.3576, interpMode=RCIM_Cubic, leaveTangent=-1.3576, tangentMode=RCTM_User, time=1, value=0}
          ]}}}
       CurveIndex=Linked() {linkedVariable {name=NPC.FortniteNPC.TimeOfDay_0-1, type=NiagaraFloat}}
      )
     )
    )) {isSetParameters=true}
   &SpriteFacingAndAlignment(
    Sprite Alignment=&TransformVector(Vector=<0 0 1>, Destination Space=::World)
    Sprite Facing=&TransformVector(Vector=<0 -1 0>, Source Space=::World, Destination Space=::Simulation)
    WriteAlignment=true
    WriteFacing=true
   ) {enabled=false}
   &SpriteFacingAndAlignment(Sprite Facing=<0 1 0>, WriteAlignment=false, WriteFacing=true) {module=SpriteFacingAndAlignment001}
   &ParticleState(Kill Particles When Lifetime Has Elapsed=false, Let Infinitely Lived Particles Die When Emitter Deactivates=true)
   &FN_Light_v3(
    Color=<rgba 0.25 0.0594 0.0519 1>
    Create Color Attribute=true
    Scale Color:NiagaraFloat=/Game/Effects/Niagara/DynamicInputs/Math/FN_Fit(
     Clamp:NiagaraBool=false
     Value:NiagaraFloat=&Sine(Normalized Angle=$Emitter.Age, Period=3, Scale=1, Bias=1)
     SourceMin:NiagaraFloat=0
     SourceMax:NiagaraFloat=1
     DestMin:NiagaraFloat=3.5
     DestMax:NiagaraFloat=5
    )
   )
  ]
  rendererClasses [
   &NiagaraMeshRendererProperties
   &NiagaraSpriteRendererProperties
   &NiagaraLightRendererProperties
  ]
  renderers [
   {rendererClass=&NiagaraMeshRendererProperties, rendererIndex=0}
   {rendererClass=&NiagaraSpriteRendererProperties, rendererIndex=1}
   {rendererClass=&NiagaraLightRendererProperties, rendererIndex=2}
  ]
 }}
summary {
 systemName=NS_Solstice_Backpack
 emitters [{
   emitterName=RadialFlare
   rendererClasses [
    &NiagaraMeshRendererProperties
    &NiagaraSpriteRendererProperties
    &NiagaraLightRendererProperties
   ]
  }]
 userVariables []
}
